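# Animation Blueprint for the Slurp Juice prop skeleton, stored with the main player's healing reaction animations.
# A looping or static default pose sequence player feeds DefaultSlot, which goes straight to the output pose.
# Montages played on that slot can therefore override the default pose.
# The event graph holds only an empty update event, and there are no variables or state machines.
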AnimBlueprint /Game/Animation/Game/MainPlayer/Reacts/Healing/SlurpJuice/SlurpJuice_Prop_AnimBP
target_skeleton=/Game/Athena/Items/Consumables/PurpleStuff/Slurp_Juice_Gadget_Skeleton
graphs [
 AnimGraph {nodes [Output Pose(Result=Slot 'DefaultSlot'(Source=Sequence Player 'SlurpJuice_Default_Pose_Prop'() {sequence=&/SlurpJuice_Default_Pose_Prop})) {layerGroup=None}]}
 EventGraph {nodes [Event BlueprintUpdateAnimation()]}
]
variables []
